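# Idle Niagara effect for the NanaEternal guitar cosmetic: flames, body glow and air glow with a light, attached to the skeletal mesh.
# Flame and glow-card particles use hand-placed offset arrays, switched by IsInstrument and IsPickaxe, which also selects FlamesUP or FlamesDown.
# System mesh scale sizes offsets and flames; speed fades flames.
# Time-of-day and front-end preview adjust brightness; Illumination001 and debugscale are disabled.
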NiagaraSystem /SparksCosmetics/Instrument/Guitar/NanaEternal/Effects/NS_Guitar_NanaEternal_IdleFX
dependencies {
 usedDataInterfaces [
  &NiagaraDataInterfaceCurve
  &NiagaraDataInterfaceArrayFloat3
  &NiagaraDataInterfaceArrayFloat
  &NiagaraDataInterfaceColorCurve
 ]
 usedDynamicInputs [
  &Multiply_Vector2D_ByFloat
  &ColorFromCurve
  &VectorLength
  /Game/Effects/Niagara/DynamicInputs/Math/FN_Fit
 ]
 usedModules [
  &SystemState
  &self:SystemSpawnScript.SystemScriptSource.SystemScriptGraph.NiagaraNodeAssignment_0.SetVariables_#1
  &self:SystemSpawnScript.SystemScriptSource.SystemScriptGraph.NiagaraNodeAssignment_5.SetVariables_#2
  &self:SystemSpawnScript.SystemScriptSource.SystemScriptGraph.NiagaraNodeAssignment_1.SetVariables_#3
  &self:SystemSpawnScript.SystemScriptSource.SystemScriptGraph.NiagaraNodeAssignment_2.SetVariables_#4
  &self:SystemSpawnScript.SystemScriptSource.SystemScriptGraph.NiagaraNodeAssignment_3.SetVariables_#5
  &self:SystemSpawnScript.SystemScriptSource.SystemScriptGraph.NiagaraNodeAssignment_4.SetVariables_#6
 ]
}
emitters {
 AirGlow_cards {
  emitterSpawnScript [/SetVariables(Emitter.SpawnCount=8) {isSetParameters=true}]
  emitterUpdateScript [
   &EmitterState()
   &SpawnBurst_Instantaneous(Spawn Count=$Emitter.SpawnCount, Use Spawn Probability=false, UseLoopCountLimit=true)
  ]
  particleSpawnScript [
   &PartitionParticles(Number Of Partitions=$Emitter.SpawnCount, Particles Per Partition=1, Partition Mode:ENiagara_PartitionMode_v2=::NewEnumerator1)
   &InitializeParticle(
    Color=<rgba 0 1 0 0.25>
    Mesh Scale Mode=::NewEnumerator4
    Position Offset=<0 0 100>
    Sprite Size Mode=::NewEnumerator3
    Uniform Sprite Size=&Multiply_Float(
     A=30
     B=&SelectFloatFromArray(
      Direct Array Index=$Particles.PartitionParticles.Partition
      Float Selection Array=DataInterface() {propertyValues {FloatData [1,1,1,1,1,1,1,1]}}
     )
    )
    UsePositionOffset=true
    Write Material Random=true
   )
   &CameraOffset(Camera Offset Amount=&SelectFloatFromArray(
     Direct Array Index=$Particles.PartitionParticles.Partition
     Float Selection Array=DataInterface() {propertyValues {FloatData [-10,-10,-10,-10,-10,200,200,-10,-10,-10,-10]}}
    )) {enabled=false}
  ]
  particleUpdateScript [
   &ParticleState(Kill Particles When Lifetime Has Elapsed=false, Let Infinitely Lived Particles Die When Emitter Deactivates=true)
   &SkeletalMeshLocation(
    MaskSpawnGroup=false
    Mesh Sampling Type=::NewEnumerator5
    Sampled Position Offset=&Multiply_VectorByFloat(
     Vector=&MakeCustomVectorFromBool(
      Bool=&SetBoolByFloatComparison(A=$User.IsInstrument, B=0)
      True Vector=&SelectVectorFromArray(
       Direct Array Index=$Particles.PartitionParticles.Partition
       Vector Selection Array=DataInterface() {propertyValues {FloatData [<12 -3.5 14>, <-11.5 -3.5 -19>, <0.344 -1 72>, <10 -3.5 -19>, <13 -3.5 -7.4>, <-13 -3.5 -7.4>, <-12 -3.5 15>, <0 0 -21>]}}
      )
      False Vector=&MakeCustomVectorFromBool(
       Bool=&SetBoolByFloatComparison(A=$User.IsPickaxe, B=0)
       True Vector=&SelectVectorFromArray(
        Direct Array Index=$Particles.PartitionParticles.Partition
        Vector Selection Array=DataInterface() {propertyValues {FloatData [<10.9 -3.5 17>, <-11 -3.5 -13>, <-0.5 -3.5 -19>, <10 -3.5 -13>, <9.5 -3.5 3>, <-11.25 -3.5 3>, <-12.8 -3.5 17>, <-1 0 67>]}}
       )
       False Vector=&SelectVectorFromArray(
        Direct Array Index=$Particles.PartitionParticles.Partition
        Vector Selection Array=DataInterface() {propertyValues {FloatData [<11 -3.5 14>, <-11.5 -3.5 -19>, <0 -1 65.1737>, <11 -3.5 -19>, <12.6636 -3.5 -6.4>, <-11 -3.5 -6.4>, <-11 -3.5 14>, <0 0 -21>]}}
       )
      )
     )
     Float=$System.MeshScale
    )
    Skeletal Mesh=$User.Skeletal Mesh
    Socket Sampling Mode=::NewEnumerator5
   )
   &ScaleSpriteSize(Scale Sprite Size Mode=::NewEnumerator0, Uniform Scale Factor=$System.MeshScale) {enabled=false}
   &ScaleColor(
    Scale Alpha=&MakeCustomFloatFromBool(Bool=$User.bIsFrontEndPreview, True Float=0.25, False Float=1)
    Scale RGB=&VectorFromFloat(Value=&FloatFromCurve(
      FloatCurve=DataInterface() {propertyValues {Curve {
         defaultValue=1
         keys [
          {interpMode=RCIM_Cubic, tangentMode=RCTM_User, time=0, value=0.7}
          {interpMode=RCIM_Cubic, tangentMode=RCTM_User, time=0.5, value=1}
          {tangentMode=RCTM_User, time=1, value=0.7}
         ]
        }}}
      CurveIndex=Linked() {linkedVariable {name=NPC.FortniteNPC.TimeOfDay_0-1, type=NiagaraFloat}}
     ))
    ScaleA=true
    ScaleRGB=true
   )
  ]
  rendererClasses [&NiagaraSpriteRendererProperties, &NiagaraMeshRendererProperties]
  renderers [
   {rendererClass=&NiagaraSpriteRendererProperties}
   {rendererClass=&NiagaraMeshRendererProperties, rendererIndex=1}
  ]
 }
 Airglow_Capsule {
  emitterUpdateScript [
   &EmitterState()
   &SpawnBurst_Instantaneous(Use Spawn Probability=false, UseLoopCountLimit=true)
  ]
  particleSpawnScript [
   &InitializeParticle(
    Color=<rgba 0 1 0 0.15>
    Lifetime=5
    Sprite Size=<60 100>
    Sprite Size Mode=::NewEnumerator4
    UsePositionOffset=false
    Write Material Random=true
   )
   &CameraOffset(Camera Offset Amount=10) {enabled=false}
  ]
  particleUpdateScript [
   &ParticleState(Kill Particles When Lifetime Has Elapsed=false, Let Infinitely Lived Particles Die When Emitter Deactivates=true)
   &SpriteFacingAndAlignment(Sprite Alignment=<0 0 1>, WriteAlignment=true, WriteFacing=false)
   &SkeletalMeshLocation(
    MaskSpawnGroup=false
    Mesh Sampling Type=::NewEnumerator5
    Sampled Position Offset=<0 -2 4>
    Skeletal Mesh=$User.Skeletal Mesh
    Socket Sampling Mode=::NewEnumerator5
   )
   &FN_Light_Attributes(
    Color=&Multiply_LinearColorByFloat(
     Linear Color=$System.Color
     Float=&MakeCustomFloatFromBool(Bool=$User.bIsFrontEndPreview, True Float=0.3, False Float=0.4)
    )
    Create Color Attribute=true
    Create Exponent Attribute=true
    Create Position Attribute=true
    Create Radius Attribute=true
    Exponent=3
    Position=&AddVectorToPosition(Position=$Particles.SkeletalMeshLocation.SampledPosition, Vector=<0 -2.5 -3.3>)
    Radius=75
    Scale Color=9
   )
   &ScaleColor(
    Scale RGB=&VectorFromFloat(Value=&FloatFromCurve(
      FloatCurve=DataInterface() {propertyValues {Curve {
         defaultValue=1
         keys [
          {interpMode=RCIM_Cubic, tangentMode=RCTM_User, time=0, value=0.7}
          {interpMode=RCIM_Cubic, tangentMode=RCTM_User, time=0.5, value=1}
          {tangentMode=RCTM_User, time=1, value=0.7}
         ]
        }}}
      CurveIndex=Linked() {linkedVariable {name=NPC.FortniteNPC.TimeOfDay_0-1, type=NiagaraFloat}}
     ))
    ScaleA=false
    ScaleRGB=true
   ) {module=ScaleColor001}
  ]
  rendererClasses [
   &NiagaraSpriteRendererProperties
   &NiagaraMeshRendererProperties
   &NiagaraLightRendererProperties
  ]
  renderers [
   {rendererClass=&NiagaraSpriteRendererProperties}
   {rendererClass=&NiagaraMeshRendererProperties, rendererIndex=1}
   {rendererClass=&NiagaraLightRendererProperties, rendererIndex=2}
  ]
 }
 FlamesDown {
  emitterUpdateScript [
   &EmitterState(Life Cycle Mode=::NewEnumerator1, Loop Behavior=::NewEnumerator1, Loop Duration Mode=::NewEnumerator1, UseLoopDelay=false)
   &SpawnBurst_Instantaneous(
    Spawn Count=8
    Spawn Probability=&MakeCustomFloatFromBool(Bool=&SetBoolByFloatComparison(A=$User.IsPickaxe, B=0), True Float=1, False Float=0)
    Use Spawn Probability=true
    UseLoopCountLimit=true
   )
   /SetVariables(Emitter.UV Flow:Vector4f=&MakeVector4(X=1, Y=1, Z=0, W=&RandomRangeFloat(Minimum=1.3, Maximum=1.7))) {isSetParameters=true}
  ]
  particleSpawnScript [
   &FN_SetSpawnGroup(Number of Spawn Groups=8)
   &InitializeParticle(
    Color=<rgba 0.2604 5 0.5232 1>
    Mesh Scale=&SelectVectorFromArray(
     Direct Array Index=&SpawnGroup()
     Vector Selection Array=DataInterface() {propertyValues {FloatData [
        <0.15 0.06 0.05>
        <0.15 0.15 0.06>
        <0.2 0.15 0.05>
        <0.15 0.15 0.06>
        <0.16 0.11 0.06>
        <0.16 0.11 0.06>
        <0.15 0.14 0.05>
        <0.22 0.06 0.06>
       ]}}
    )
    Mesh Scale Mode=::NewEnumerator4
    Sprite Size Mode=::NewEnumerator3
    Uniform Sprite Size=5
    UsePositionOffset=false
    Write Material Random=true
   )
   &CameraOffset(Camera Offset Amount=-5)
   &InitialMeshOrientation(
    Rotation=&SelectVectorFromArray(
     Direct Array Index=$Particles.SpawnGroup
     Vector Selection Array=DataInterface() {propertyValues {FloatData [<-0.5 0 0>, <-0.5 0 0>, <-0.5 0 0>, <-0.5 0 0>, <-0.5 0 0>, <-0.5 0 0>, <-0.5 0 0>, <-0.5 0 0>]}}
    )
    Use Orientation Vector=true
    Use Rotation Vector=true
   ) {module=InitialMeshOrientation001}
   &InitialMeshOrientation(
    Rotation=&SelectVectorFromArray(
     Direct Array Index=$Particles.SpawnGroup
     Vector Selection Array=DataInterface() {propertyValues {FloatData [<0 -0.008 0>, <0 0 0>, <0 0 0>, <0 0 0>, <0 -0.06 0>, <0 0.06 0>, <0 0.008 0>, <0 0 0>]}}
    )
    Use Orientation Vector=true
    Use Rotation Vector=true
   )
  ]
  particleUpdateScript [
   &ParticleState(Kill Particles When Lifetime Has Elapsed=false, Let Infinitely Lived Particles Die When Emitter Deactivates=true)
   &DynamicMaterialParameters(
    Index 0 Param 1=&RandomRangeFloat(Minimum=0, Maximum=1)
    Index 0 Param 2=&RandomRangeFloat(Minimum=0, Maximum=1)
    Index 0 Param 3=&Waveform(
     Global Phase=&Add_Float(A=$Emitter.Age, B=&ReturnNormalizedExecIndex())
     "Waveform [1]"=::NewEnumerator0
     "[1] Frequency"=1.5
     "[1] Amplitude Min/Max"=<5 8>
    )
    Index 0 Param 4=1
    Module.Index0Float0Write=true
    Module.Index0Float1Write=true
    Module.Index0Float2Write=true
    Module.Index0Float3Write=true
    Module.Index1Float0Write=true
    Module.Index1Float1Write=true
    Module.Index1Float2Write=true
    Module.Index1Float3Write=true
    Module.Index2Float0Write=true
    Module.Index2Float1Write=true
    Module.Index2Float2Write=true
    Module.Index2Float3Write=true
    Module.Index3Float0Write=true
    Module.Index3Float1Write=true
    Module.Index3Float2Write=true
    Module.Index3Float3Write=true
    Write Parameter Index 1=::NewEnumerator1
    Write Parameter Index 2=::NewEnumerator1
    Write Parameter Index 3=::NewEnumerator1
   )
   &SkeletalMeshLocation(
    MaskSpawnGroup=false
    Mesh Sampling Type=::NewEnumerator5
    Sampled Position Offset=&Multiply_VectorByFloat(
     Vector=&SelectVectorFromArray(
      Direct Array Index=&SpawnGroup()
      Vector Selection Array=DataInterface() {propertyValues {FloatData [<10.9 -3.5 19>, <-11 -3.5 -12>, <-0.5 -3.5 -17.5>, <10 -3.5 -12.8>, <9.5 -3.5 1.5>, <-11.25 -3.5 1.5>, <-12.8 -3.5 19>, <-1 0 69.929>]}}
     )
     Float=$System.MeshScale
    )
    Skeletal Mesh=$User.Skeletal Mesh
    Socket Sampling Mode=::NewEnumerator5
   )
   &ScaleSpriteSize(Scale Sprite Size Mode=::NewEnumerator0, Uniform Scale Factor=$System.MeshScale)
   &ScaleMeshSize(Scale Factor=&VectorFromFloat(Value=$System.MeshScale))
   &ScaleColor(
    Scale Alpha=&Multiply_Float(
     A=&SelectFloatFromArray(
      Direct Array Index=&SpawnGroup()
      Float Selection Array=DataInterface() {propertyValues {FloatData [0.7,1,1,1,0.7,0.7,0.7,0.7]}}
     )
     B=&MakeCustomFloatFromBool(Bool=$User.bIsFrontEndPreview, True Float=0.8, False Float=1)
    )
    Scale RGB=&Multiply_VectorByFloat(Vector=&MakeVectorFromLinearColorRGB(Color=$System.TOD Color), Float=2)
    ScaleA=true
    ScaleRGB=true
   ) {module=ScaleColor002}
   &InterpolateOverTime(Rate Of Change=15, Target Vector Value:NiagaraPosition=$Particles.Position, Variable Type=::NewEnumerator2)
   /SetVariables(Particles.Interpolate Dynamic=&MakeFloat4FromVectorAndFloat("Vector (RGB)"=$Output.InterpolateOverTime.Moving Average, "Float (Alpha)"=1)) {isSetParameters=true, module=SetVariables_#7}
   &ScaleColor(Scale Alpha=&Lerp_Float(A=1, B=0.25, Alpha=$System.NormalizedSpeed), ScaleA=true, ScaleRGB=false) {module=ScaleColor001}
  ]
  rendererClasses [&NiagaraMeshRendererProperties]
  renderers [
   {rendererClass=&NiagaraMeshRendererProperties, rendererIndex=0}
   {rendererClass=&NiagaraMeshRendererProperties, rendererIndex=1}
  ]
 }
 FlamesUP {
  emitterUpdateScript [
   &EmitterState(Life Cycle Mode=::NewEnumerator1, Loop Behavior=::NewEnumerator1, Loop Duration Mode=::NewEnumerator1, UseLoopDelay=false)
   &SpawnBurst_Instantaneous(
    Spawn Count=7
    Spawn Probability=&MakeCustomFloatFromBool(Bool=&SetBoolByFloatComparison(A=$User.IsPickaxe, B=0), True Float=0, False Float=1)
    Use Spawn Probability=true
    UseLoopCountLimit=true
   )
   /SetVariables(Emitter.UV Flow:Vector4f=&MakeVector4(X=1, Y=1, Z=0, W=&RandomRangeFloat(Minimum=1, Maximum=1.7))) {isSetParameters=true}
  ]
  particleSpawnScript [
   &FN_SetSpawnGroup(Number of Spawn Groups=7)
   &InitializeParticle(
    Color=<rgba 0.2604 5 0.5232 1>
    Mesh Scale=&SelectVectorFromArray(
     Direct Array Index=&SpawnGroup()
     Vector Selection Array=DataInterface() {propertyValues {FloatData [<0.06 0.15 0.062>, <0.1 0.12 0.06>, <0.07 0.15 0.055>, <0.1 0.12 0.06>, <0.1 0.17 0.06>, <0.1 0.17 0.06>, <0.06 0.15 0.062>]}}
    )
    Mesh Scale Mode=::NewEnumerator4
    Sprite Size Mode=::NewEnumerator3
    Uniform Sprite Size=5
    UsePositionOffset=false
    Write Material Random=true
   )
   &CameraOffset(Camera Offset Amount=-5)
   &InitialMeshOrientation(
    Rotation=&SelectVectorFromArray(
     Direct Array Index=$Particles.SpawnGroup
     Vector Selection Array=DataInterface() {propertyValues {FloatData [<0.03 0.03 0.25>, <-0.05 -0.05 0.25>, <0 0 0.25>, <0.05 0.05 0.25>, <0.02 0.02 0.25>, <-0.02 -0.02 0.25>, <-0.03 -0.03 0.25>]}}
    )
    Use Orientation Vector=true
    Use Rotation Vector=true
   ) {module=InitialMeshOrientation001}
  ]
  particleUpdateScript [
   &ParticleState(Kill Particles When Lifetime Has Elapsed=false, Let Infinitely Lived Particles Die When Emitter Deactivates=true)
   &DynamicMaterialParameters(
    Index 0 Param 1=&RandomRangeFloat(Minimum=0, Maximum=1)
    Index 0 Param 2=&RandomRangeFloat(Minimum=0, Maximum=1)
    Index 0 Param 3=&Waveform(
     Global Phase=&Add_Float(A=$Emitter.Age, B=&ReturnNormalizedExecIndex())
     "Waveform [1]"=::NewEnumerator0
     "[1] Frequency"=1.5
     "[1] Amplitude Min/Max"=<5 8>
    )
    Index 0 Param 4=1
    Module.Index0Float0Write=true
    Module.Index0Float1Write=true
    Module.Index0Float2Write=true
    Module.Index0Float3Write=true
    Module.Index1Float0Write=true
    Module.Index1Float1Write=true
    Module.Index1Float2Write=true
    Module.Index1Float3Write=true
    Module.Index2Float0Write=true
    Module.Index2Float1Write=true
    Module.Index2Float2Write=true
    Module.Index2Float3Write=true
    Module.Index3Float0Write=true
    Module.Index3Float1Write=true
    Module.Index3Float2Write=true
    Module.Index3Float3Write=true
    Write Parameter Index 1=::NewEnumerator1
    Write Parameter Index 2=::NewEnumerator1
    Write Parameter Index 3=::NewEnumerator1
   )
   &SkeletalMeshLocation(
    MaskSpawnGroup=false
    Mesh Sampling Type=::NewEnumerator5
    Sampled Position Offset=&Multiply_VectorByFloat(
     Vector=&MakeCustomVectorFromBool(
      Bool=&SetBoolByFloatComparison(A=$User.IsInstrument, B=0)
      True Vector=&SelectVectorFromArray(
       Direct Array Index=&SpawnGroup()
       Vector Selection Array=DataInterface() {propertyValues {FloatData [<12.75 -3.5 10.9219>, <-10.8282 -3.5 -19>, <2.25 -1 69>, <13.0879 -3.5 -19>, <13.4113 -3.5 -7.4>, <-9.9212 -3.5 -7.4>, <-9.25 -3.5 12>]}}
      )
      False Vector=&SelectVectorFromArray(
       Direct Array Index=&SpawnGroup()
       Vector Selection Array=DataInterface() {propertyValues {FloatData [<11 -3.5 13>, <-11.5 -3.5 -19>, <0 -1 70>, <11 -3.5 -19>, <12.6636 -3.5 -7.4>, <-11 -3.5 -7.4>, <-11 -3.5 13>]}}
      )
     )
     Float=$System.MeshScale
    )
    Skeletal Mesh=$User.Skeletal Mesh
    Socket Sampling Mode=::NewEnumerator5
   )
   &ScaleSpriteSize(Scale Sprite Size Mode=::NewEnumerator0, Uniform Scale Factor=$System.MeshScale)
   &ScaleMeshSize(Scale Factor=&VectorFromFloat(Value=$System.MeshScale))
   &ScaleColor(
    Scale Alpha=&Multiply_Float(
     A=&SelectFloatFromArray(
      Direct Array Index=&SpawnGroup()
      Float Selection Array=DataInterface() {propertyValues {FloatData [1,0.4,1.1,0.4,0.7,0.7,1]}}
     )
     B=&MakeCustomFloatFromBool(Bool=$User.bIsFrontEndPreview, True Float=1, False Float=1)
    )
    Scale RGB=&Multiply_VectorByFloat(Vector=&MakeVectorFromLinearColorRGB(Color=$System.TOD Color), Float=2)
    ScaleA=true
    ScaleRGB=true
   ) {module=ScaleColor002}
   &InterpolateOverTime(Rate Of Change=15, Target Vector Value:NiagaraPosition=$Particles.Position, Variable Type=::NewEnumerator2)
   /SetVariables(Particles.Interpolate Dynamic=&MakeFloat4FromVectorAndFloat("Vector (RGB)"=$Output.InterpolateOverTime.Moving Average, "Float (Alpha)"=0.75)) {isSetParameters=true, module=SetVariables_#7}
   &ScaleColor(Scale Alpha=&Lerp_Float(A=1, B=0.6, Alpha=$System.NormalizedSpeed), ScaleA=true, ScaleRGB=false) {module=ScaleColor001}
  ]
  rendererClasses [&NiagaraMeshRendererProperties]
  renderers [
   {rendererClass=&NiagaraMeshRendererProperties, rendererIndex=0}
   {rendererClass=&NiagaraMeshRendererProperties, rendererIndex=1}
  ]
 }
 Glow {
  emitterUpdateScript [
   &EmitterState()
   &SpawnBurst_Instantaneous(Use Spawn Probability=false, UseLoopCountLimit=true)
  ]
  particleSpawnScript [&InitializeParticle(
    Color=$System.Color
    Lifetime=9
    Mesh Scale Mode=::NewEnumerator3
    Sprite Size Mode=::NewEnumerator3
    UsePositionOffset=false
    Write Material Random=true
   )]
  particleUpdateScript [
   &ParticleState(
    Kill Particles When Lifetime Has Elapsed=false
    Let Infinitely Lived Particles Die When Emitter Deactivates=true
    Loop Particles Lifetime=true
   )
   &SolveForcesAndVelocity(Clamp Velocity=false, Limit Acceleration=false)
   &DynamicMaterialParameters(
    Index 0 Param 1=&MakeCustomFloatFromBool(
     Bool=&SetBoolByFloatComparison(A=$User.IsPickaxe, B=0)
     True Float=&FloatFromCurve(
      FloatCurve=DataInterface() {propertyValues {Curve {defaultValue=1, keys [{tangentMode=RCTM_User, time=0, value=0}, {tangentMode=RCTM_User, time=1, value=1}]}}}
      CurveIndex=&Frac_Float(Float=&Multiply_Float(A=0.6, B=$Engine.Time))
     )
     False Float=&FloatFromCurve(
      FloatCurve=DataInterface() {propertyValues {Curve {defaultValue=0, keys [{tangentMode=RCTM_User, time=0, value=1}, {tangentMode=RCTM_User, time=1, value=0}]}}}
      CurveIndex=&Frac_Float(Float=&Multiply_Float(A=0.6, B=$Engine.Time))
     )
    )
    Module.Index0Float0Write=true
    Module.Index0Float1Write=true
    Module.Index0Float2Write=true
    Module.Index0Float3Write=true
   )
   /SetVariables(
    Particles.ColorBody:LinearColor=&MakeLinearColorFromVectorAndFloat(
     "Vector (RGB)"=&Multiply_Vector(
      A=&Multiply_Vector(
       A=&MakeVectorFromLinearColorRGB(Color=$System.Color)
       B=&MakeVectorFromLinearColorRGB(Color=$System.TOD Color)
      )
      B=&VectorFromFloat(Value=2)
     )
     "Float (Alpha)"=&Multiply_Float(
      A=&MakeCustomFloatFromBool(Bool=$User.bIsFrontEndPreview, True Float=0.45, False Float=1)
      B=&FloatFromCurve(
       FloatCurve=DataInterface() {propertyValues {Curve {keys [
           {arriveTangent=0.1117, interpMode=RCIM_Cubic, leaveTangent=0.1117, tangentMode=RCTM_User, time=0, value=1}
           {
            arriveTangent=-0.0076
            arriveTangentWeight=0.0963
            interpMode=RCIM_Cubic
            leaveTangent=-0.0076
            leaveTangentWeight=0.1693
            tangentMode=RCTM_User
            time=0.3
            value=1.25
           }
           {arriveTangentWeight=0.0722, interpMode=RCIM_Cubic, leaveTangentWeight=0.1444, tangentMode=RCTM_User, time=0.6, value=1.25}
           <at 1 1 RCIM_Cubic>
          ]}}}
       CurveIndex=Linked() {linkedVariable {name=NPC.FortniteNPC.TimeOfDay_0-1, type=NiagaraFloat}}
       Scale Curve=2
      )
     )
    )
    Particles.ColorNeck:LinearColor=&MakeLinearColorFromVectorAndFloat(
     "Vector (RGB)"=&Multiply_Vector(
      A=&Multiply_Vector(
       A=&MakeVectorFromLinearColorRGB(Color=$System.Color)
       B=&MakeVectorFromLinearColorRGB(Color=$System.TOD Color)
      )
      B=&VectorFromFloat(Value=2)
     )
     "Float (Alpha)"=&Multiply_Float(
      A=&MakeCustomFloatFromBool(Bool=$User.bIsFrontEndPreview, True Float=0.5, False Float=1)
      B=&Multiply_Float(
       A=0.75
       B=&FloatFromCurve(
        FloatCurve=DataInterface() {propertyValues {Curve {keys [
            {arriveTangent=0.1117, interpMode=RCIM_Cubic, leaveTangent=0.1117, tangentMode=RCTM_User, time=0, value=1}
            {
             arriveTangent=-0.0076
             arriveTangentWeight=0.0963
             interpMode=RCIM_Cubic
             leaveTangent=-0.0076
             leaveTangentWeight=0.1693
             tangentMode=RCTM_User
             time=0.3
             value=1.25
            }
            {arriveTangentWeight=0.0722, interpMode=RCIM_Cubic, leaveTangentWeight=0.1444, tangentMode=RCTM_User, time=0.6, value=1.25}
            <at 1 1 RCIM_Cubic>
           ]}}}
        CurveIndex=Linked() {linkedVariable {name=NPC.FortniteNPC.TimeOfDay_0-1, type=NiagaraFloat}}
        Scale Curve=3
       )
      )
     )
    )
    Particles.ColorTop:LinearColor=&MakeLinearColorFromVectorAndFloat(
     "Vector (RGB)"=&Multiply_Vector(
      A=&Multiply_Vector(
       A=&MakeVectorFromLinearColorRGB(Color=$System.Color)
       B=&MakeVectorFromLinearColorRGB(Color=$System.TOD Color)
      )
      B=&VectorFromFloat(Value=2)
     )
     "Float (Alpha)"=&Multiply_Float(
      A=&MakeCustomFloatFromBool(
       Bool=$User.bIsFrontEndPreview
       True Float=0.75
       False Float=&MakeCustomFloatFromBool(Bool=&SetBoolByFloatComparison(A=$User.IsPickaxe, B=0), True Float=0.75, False Float=1)
      )
      B=&FloatFromCurve(
       FloatCurve=DataInterface() {propertyValues {Curve {keys [
           {arriveTangent=0.1117, interpMode=RCIM_Cubic, leaveTangent=0.1117, tangentMode=RCTM_User, time=0, value=1}
           {
            arriveTangent=-0.0076
            arriveTangentWeight=0.0963
            interpMode=RCIM_Cubic
            leaveTangent=-0.0076
            leaveTangentWeight=0.1693
            tangentMode=RCTM_User
            time=0.2
            value=1.25
           }
           {arriveTangentWeight=0.0722, interpMode=RCIM_Cubic, leaveTangentWeight=0.1444, tangentMode=RCTM_User, time=0.6, value=1.25}
           <at 1 1 RCIM_Cubic>
          ]}}}
       CurveIndex=Linked() {linkedVariable {name=NPC.FortniteNPC.TimeOfDay_0-1, type=NiagaraFloat}}
      )
     )
    )
   ) {isSetParameters=true}
  ]
  rendererClasses [&NiagaraMeshRendererProperties]
  renderers [
   {rendererClass=&NiagaraMeshRendererProperties, rendererIndex=0}
   {rendererClass=&NiagaraMeshRendererProperties, rendererIndex=1}
   {rendererClass=&NiagaraMeshRendererProperties, rendererIndex=2}
   {rendererClass=&NiagaraMeshRendererProperties, rendererIndex=3}
   {rendererClass=&NiagaraMeshRendererProperties, rendererIndex=4}
  ]
 }
 Illumination001 {
  bEnabled=false
  emitterUpdateScript [
   &EmitterState(
    Inactive Response=::NewEnumerator1
    Life Cycle Mode=::NewEnumerator1
    Loop Behavior=::NewEnumerator1
    Loop Duration Mode=::NewEnumerator1
    UseLoopDelay=false
   )
   &SpawnBurst_Instantaneous(Use Spawn Probability=false, UseLoopCountLimit=true)
  ]
  particleSpawnScript [
   &InitializeParticle(
    Color=&Multiply_LinearColorByFloat(
     Linear Color=$System.Color
     Float=&MakeCustomFloatFromBool(
      Bool=&Boolean_OrOperation(Bool1=$User.bIsFrontEnd, Bool2=&Boolean_OrOperation(Bool1=$User.bIsFrontEndPreview, Bool2=$User.IsFrontEnd))
      True Float=0.3
      False Float=0.4
     )
    )
    Sprite Size=<60 100>
    Sprite Size Mode=::NewEnumerator4
    UsePositionOffset=true
    Write Material Random=true
   )
   &CameraOffset(Camera Offset Amount=20) {enabled=false}
   &DynamicMaterialParameters(
    Index 0 Param 1=2
    Index 0 Param 2=25
    Index 0 Param 3=1.1
    Module.Index0Float0Write=true
    Module.Index0Float1Write=true
    Module.Index0Float2Write=true
    Module.Index0Float3Write=true
   )
  ]
  particleUpdateScript [
   &ParticleState(Kill Particles When Lifetime Has Elapsed=false, Let Infinitely Lived Particles Die When Emitter Deactivates=true)
   &SkeletalMeshLocation(
    MaskSpawnGroup=false
    Mesh Sampling Type=::NewEnumerator5
    Sampled Position Offset=&Multiply_VectorByFloat(Vector=<0 -2 0>, Float=$System.MeshRemapScale)
    Sampled Position Offset Coordinate Space:ENiagaraSkelMeshTransforms=::NewEnumerator2
    Skeletal Mesh=$User.Skeletal Mesh
    Socket Sampling Mode=::NewEnumerator5
   )
   &SpriteFacingAndAlignment(Sprite Alignment=<0 0 1>, WriteAlignment=true, WriteFacing=false)
   &FN_Light_Attributes(
    Color=$Particles.Color
    Create Color Attribute=true
    Create Exponent Attribute=true
    Create Position Attribute=true
    Create Radius Attribute=true
    Exponent=3
    Position=&AddVectorToPosition(Position=$Particles.SkeletalMeshLocation.SampledPosition, Vector=<0 -2.5 -3.3>)
    Radius=75
    Scale Color=9
   )
   &ScaleColor(
    Scale Alpha=&MakeCustomFloatFromBool(
     Bool=&Boolean_OrOperation(Bool1=&Boolean_OrOperation(Bool1=$User.bIsFrontEnd, Bool2=$User.IsFrontEnd), Bool2=$User.bIsFrontEndPreview)
     True Float=0.5
     False Float=1
    )
    Scale RGB=&MakeVectorFromLinearColorRGB(Color=$System.TOD Color)
    ScaleA=true
    ScaleRGB=true
   )
  ]
  rendererClasses [&NiagaraSpriteRendererProperties, &NiagaraLightRendererProperties]
  renderers [
   {rendererClass=&NiagaraSpriteRendererProperties}
   {rendererClass=&NiagaraLightRendererProperties, rendererIndex=1}
   {rendererClass=&NiagaraSpriteRendererProperties, rendererIndex=2}
  ]
 }
 debugscale {
  bEnabled=false
  emitterUpdateScript [
   &EmitterState()
   &SpawnBurst_Instantaneous(Use Spawn Probability=false, UseLoopCountLimit=false)
  ]
  particleSpawnScript [&InitializeParticle(
    Color Mode=::NewEnumerator0
    Lifetime=5
    Sprite Size Mode=::NewEnumerator3
    Uniform Sprite Size=100
    UsePositionOffset=false
    Write Material Random=true
   )]
  particleUpdateScript [
   &ParticleState(Kill Particles When Lifetime Has Elapsed=false)
   &DynamicMaterialParameters(
    Index 0 Param 1=&MakeFloatFromVector(Vector=$Engine.Owner.Scale, Channel=::NewEnumerator0)
    Module.Index0Float0Write=true
    Module.Index0Float1Write=true
    Module.Index0Float2Write=true
    Module.Index0Float3Write=true
   )
  ]
  rendererClasses [&NiagaraSpriteRendererProperties]
  renderers [{rendererClass=&NiagaraSpriteRendererProperties}]
 }
}
summary {
 systemName=NS_Guitar_NanaEternal_IdleFX
 emitters [
  {emitterName=Glow, rendererClasses [&NiagaraMeshRendererProperties]}
  {bEnabled=false, emitterName=debugscale, rendererClasses [&NiagaraSpriteRendererProperties]}
  {emitterName=FlamesUP, rendererClasses [&NiagaraMeshRendererProperties]}
  {emitterName=FlamesDown, rendererClasses [&NiagaraMeshRendererProperties]}
  {
   bEnabled=false
   emitterName=Illumination001
   rendererClasses [&NiagaraSpriteRendererProperties, &NiagaraLightRendererProperties]
  }
  {emitterName=AirGlow_cards, rendererClasses [&NiagaraSpriteRendererProperties, &NiagaraMeshRendererProperties]}
  {
   emitterName=Airglow_Capsule
   rendererClasses [
    &NiagaraSpriteRendererProperties
    &NiagaraMeshRendererProperties
    &NiagaraLightRendererProperties
   ]
  }
 ]
 userVariables [
  User.bIsFrontEnd {type=NiagaraBool}
  User.bIsFrontEndPreview {type=NiagaraBool}
  User.IsFrontEnd {type=NiagaraBool}
  User.IsInstrument {type=NiagaraFloat, defaultValue=0, description=""}
  User.IsPickaxe {type=NiagaraFloat, defaultValue=0, description=""}
  User.Skeletal Mesh {
   type=NiagaraDataInterfaceSkeletalMesh
   defaultValue {
    struct=NiagaraExt_VariableValue_DataInterface
    value {dataInterface=&self:NiagaraDataInterfaceSkeletalMesh_0}
   }
  }
 ]
}
