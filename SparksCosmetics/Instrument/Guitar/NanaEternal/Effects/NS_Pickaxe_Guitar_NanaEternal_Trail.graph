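# Niagara swing trail for the Nana Eternal guitar pickaxe cosmetic, built from three ribbon emitters.
# Two glow emitters use AnimTrail modules to make short-lived ribbons that fade and narrow over life with a yellow-green to dark green tint.
# WispTrail samples the skeletal mesh for green ribbons that widen while curl noise breaks them up, with dynamic material parameters.
# Front-end checks dim all three, and time of day also feeds the wisp color.
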NiagaraSystem /SparksCosmetics/Instrument/Guitar/NanaEternal/Effects/NS_Pickaxe_Guitar_NanaEternal_Trail
dependencies {
 usedDataInterfaces [&NiagaraDataInterfaceCurve, &NiagaraDataInterfaceColorCurve]
 usedDynamicInputs []
 usedModules [
  &self:SystemSpawnScript.SystemScriptSource.SystemScriptGraph.NiagaraNodeAssignment_0.SetVariables_#1
  &SystemState
 ]
}
emitters {
 GLowTrail_CameraFacing {
  emitterUpdateScript [
   &EmitterState(Life Cycle Mode=::NewEnumerator1, Loop Behavior=::NewEnumerator1, Loop Duration=0.8, UseLoopDelay=false)
   &FN_SpawnAnimTrail()
  ]
  particleSpawnScript [
   &FN_InitializeAnimTrailBoneIndecies()
   &FN_InitializeAnimTrail(AnimTrailWidth=0.5, Color=<rgba 1 1 1 0.75>, Lifetime=0.25)
  ]
  particleUpdateScript [
   &ParticleState()
   &FN_UpdateAnimTrail(SkeletalMesh=$User.SkeletalMesh)
   &ScaleColor(
    Scale Alpha=&FloatFromCurve(
     FloatCurve=DataInterface() {propertyValues {Curve {keys [
         {arriveTangent=-2.3976, interpMode=RCIM_Cubic, leaveTangent=-2.3976, tangentMode=RCTM_User, time=0, value=1}
         {arriveTangent=-1, interpMode=RCIM_Cubic, leaveTangent=-1, time=0.15, value=0.2}
         {arriveTangent=0.0495, interpMode=RCIM_Cubic, leaveTangent=0.0495, tangentMode=RCTM_User, time=1, value=0}
        ]}}}
     Scale Curve=&FN_FrontEndFloatModifier(Gameplay=1, Front End=0.6)
    )
    Scale RGB=&MakeVectorFromLinearColorRGB(Color=&ColorFromCurve(ColorCurve=DataInterface() {propertyValues {
        AlphaCurve {keys []}
        BlueCurve {keys [<at 0 0 RCIM_Cubic>,<at 0.871 0.0496>]}
        GreenCurve {keys [<at 0 1 RCIM_Cubic>,<at 0.871 0.2345>]}
        RedCurve {keys [<at 0 0.7637 RCIM_Cubic>,<at 0.871 0.0001>]}
       }}))
    ScaleA=true
    ScaleRGB=true
   )
   &ScaleRibbonWidth(Ribbon Width Scale=&FloatFromCurve(FloatCurve=DataInterface() {propertyValues {Curve {
        defaultValue=1
        keys [
         {interpMode=RCIM_Cubic, tangentMode=RCTM_User, time=0, value=1}
         {arriveTangent=-3.5594, interpMode=RCIM_Cubic, leaveTangent=-3.5594, tangentMode=RCTM_User, time=1, value=0}
        ]
       }}}))
  ]
  rendererClasses [&NiagaraRibbonRendererProperties]
  renderers [{rendererClass=&NiagaraRibbonRendererProperties}]
 }
 GlowTrail {
  emitterUpdateScript [
   &EmitterState(Life Cycle Mode=::NewEnumerator1, Loop Behavior=::NewEnumerator1, Loop Duration=0.22, UseLoopDelay=true)
   &FN_SpawnAnimTrail_v2()
  ]
  particleSpawnScript [
   /Game/Effects/Niagara/Modules/AnimTrail/FN_InitializeAnimTrail/FN_InitializeAnimTrailBoneIndecies(First Bone or Socket Index=1)
   /Game/Effects/Niagara/Modules/AnimTrail/FN_InitializeAnimTrail/FN_InitializeAnimTrail(AnimTrailWidth=1, Color=<rgba 1 1 1 0.75>, Lifetime=0.3)
   &SystemLocation(MaskSpawnGroup=false)
  ]
  particleUpdateScript [
   &ScaleColor(
    Scale Alpha=&FloatFromCurve(
     FloatCurve=DataInterface() {propertyValues {Curve {keys [
         {arriveTangent=-2.3976, interpMode=RCIM_Cubic, leaveTangent=-2.3976, tangentMode=RCTM_User, time=0, value=1}
         {arriveTangent=-1, interpMode=RCIM_Cubic, leaveTangent=-1, time=0.15, value=0.2}
         {arriveTangent=0.0495, interpMode=RCIM_Cubic, leaveTangent=0.0495, tangentMode=RCTM_User, time=1, value=0}
        ]}}}
     Scale Curve=&FN_FrontEndFloatModifier(Gameplay=1, Front End=0.6)
    )
    Scale RGB=&MakeVectorFromLinearColorRGB(Color=&ColorFromCurve(ColorCurve=DataInterface() {propertyValues {
        AlphaCurve {keys []}
        BlueCurve {keys [<at 0 0 RCIM_Cubic>,<at 0.871 0.0496>]}
        GreenCurve {keys [<at 0 1 RCIM_Cubic>,<at 0.871 0.2345>]}
        RedCurve {keys [<at 0 0.7637 RCIM_Cubic>,<at 0.871 0.0001>]}
       }}))
    ScaleA=true
    ScaleRGB=true
   )
   &ScaleRibbonWidth(Ribbon Width Scale=&FloatFromCurve(FloatCurve=DataInterface() {propertyValues {Curve {
        defaultValue=1
        keys [
         {interpMode=RCIM_Cubic, tangentMode=RCTM_User, time=0, value=1}
         {arriveTangent=-3.5594, interpMode=RCIM_Cubic, leaveTangent=-3.5594, tangentMode=RCTM_User, time=1, value=0}
        ]
       }}}))
   &FN_UpdateAnimTrail_v2()
  ]
  rendererClasses [&NiagaraRibbonRendererProperties]
  renderers [{rendererClass=&NiagaraRibbonRendererProperties}]
 }
 WispTrail {
  emitterUpdateScript [
   &EmitterState(Life Cycle Mode=::NewEnumerator1, Loop Behavior=::NewEnumerator1, Loop Delay=0.005, Loop Duration=1, UseLoopDelay=true)
   &SpawnRate(SpawnRate=45, Use Spawn Probability=false)
  ]
  particleSpawnScript [
   &InitializeRibbon(
    Color=<rgba 0.0879 0.5651 0.169 0.7>
    Lifetime=0.3
    Module.Write Color=true
    Module.Write Lifetime=true
    Module.Write Mass=false
    Module.Write Material Random=true
    Module.Write Position=false
    Module.Write RibbonTwist=false
    Module.Write RibbonWidth=true
    Ribbon Width=30
   )
   &SkeletalMeshLocation(
    Inherit Velocity=false
    MaskSpawnGroup=false
    Mesh Sampling Type=::NewEnumerator5
    Sampled Position Offset=<0 0 10>
    Sampled Position Offset Coordinate Space:ENiagaraMeshTransforms=::NewEnumerator3
    Skeletal Mesh=$User.SkeletalMesh
    Socket Sampling Mode=::NewEnumerator5
   )
   &AddVelocity(Velocity=&RandomVector(Vector Scale=&FloatFromCurve(FloatCurve=DataInterface() {propertyValues {Curve {keys [<at 0 0>,<at 1 1>]}}}, Scale Curve=1000))) {enabled=false}
  ]
  particleUpdateScript [
   &ParticleState()
   &CurlNoiseForce(
    Bias Noise=false
    Noise Frequency=300
    Noise Strength=&FloatFromCurve(
     FloatCurve=DataInterface() {propertyValues {Curve {
        defaultValue=1
        keys [
         {arriveTangent=4.1492, interpMode=RCIM_Cubic, leaveTangent=4.1492, tangentMode=RCTM_User, time=0.5, value=0}
         {interpMode=RCIM_Cubic, tangentMode=RCTM_User, time=0.9, value=1}
        ]
       }}}
     Scale Curve=5000
    )
    Pan Noise=false
   )
   &SolveForcesAndVelocity(Clamp Velocity=false, Limit Acceleration=false)
   &Color(
    Color=&ColorFromCurve(ColorCurve=DataInterface() {propertyValues {
       AlphaCurve {keys [<at 0.7149 2 RCIM_Cubic>,<at 0.9985 0>]}
       BlueCurve {keys [<at 0.0057 0.0189>,<at 0.2591 0>,<at 1 0.0156>]}
       GreenCurve {keys [<at 0.0057 0>,<at 0.2591 0>,<at 1 0.0026>]}
       RedCurve {keys [<at 0.0057 0.0039>,<at 0.2591 0>,<at 1 0>]}
      }})
    Scale Alpha=0.1
   ) {enabled=false}
   &ScaleColor(
    Scale Alpha=&Multiply_Float(
     A=&FloatFromCurve(
      FloatCurve=DataInterface() {propertyValues {Curve {keys [
          {interpMode=RCIM_Cubic, leaveTangent=-2.386, leaveTangentWeight=0.4326, tangentMode=RCTM_Break, time=0.2, value=1}
          {arriveTangent=-1.4154, arriveTangentWeight=0.2986, interpMode=RCIM_Cubic, leaveTangent=-1.4154, tangentMode=RCTM_User, time=1, value=0}
         ]}}}
      Scale Curve=&FloatFromCurve(
       FloatCurve=DataInterface() {propertyValues {Curve {keys [<at 0 1>,<at 0.497 1 RCIM_Cubic>,<at 1 1>]}}}
       CurveIndex=Linked() {linkedVariable {name=NPC.FortniteNPC.TimeOfDay_0-1, type=NiagaraFloat}}
      )
     )
     B=&MakeCustomFloatFromBool(Bool=$User.bIsFrontEndPreview, True Float=0.5, False Float=1)
    )
    Scale RGB=&Multiply_Vector(
     A=&MakeVectorFromLinearColorRGB(Color=$System.TOD Color)
     B=&MakeVectorFromLinearColorRGB(Color=&ColorFromCurve(ColorCurve=DataInterface() {propertyValues {
         AlphaCurve {keys [<at 0 1 RCIM_Cubic>]}
         BlueCurve {keys [<at 0.0901 1 RCIM_Cubic>,<at 0.6333 0.0039 RCIM_Cubic>]}
         GreenCurve {keys [<at 0.0901 1 RCIM_Cubic>,<at 0.6333 0.01 RCIM_Cubic>]}
         RedCurve {keys [<at 0.0901 1 RCIM_Cubic>,<at 0.6333 0.0042 RCIM_Cubic>]}
        }}))
    )
    ScaleA=true
    ScaleRGB=true
   ) {module=ScaleColor002}
   &ScaleRibbonWidth(Ribbon Width Scale=&FloatFromCurve(FloatCurve=DataInterface() {propertyValues {Curve {keys [
         {arriveTangent=2.6972, interpMode=RCIM_Cubic, leaveTangent=2.6972, leaveTangentWeight=0.2293, tangentMode=RCTM_User, time=0.1, value=1}
         {arriveTangent=4.8659, interpMode=RCIM_Cubic, leaveTangent=4.8659, tangentMode=RCTM_User, time=0.75, value=2.75}
         {arriveTangent=24.818, interpMode=RCIM_Cubic, leaveTangent=24.818, tangentMode=RCTM_User, time=1, value=6}
        ]}}}))
   &DynamicMaterialParameters(
    Index 0 Param 1=&FloatFromCurve(
     FloatCurve=DataInterface() {propertyValues {Curve {
        defaultValue=0
        keys [
         {arriveTangent=1.5964, interpMode=RCIM_Cubic, leaveTangent=1.5964, tangentMode=RCTM_User, time=0, value=0}
         {arriveTangent=0.7982, interpMode=RCIM_Cubic, leaveTangent=0.7982, tangentMode=RCTM_User, time=1, value=1}
        ]
       }}}
     Scale Curve=0.75
    )
    Index 0 Param 3=-0.8
    Index 0 Param 4=&FloatFromCurve(FloatCurve=DataInterface() {propertyValues {Curve {keys [
         {interpMode=RCIM_Cubic, tangentMode=RCTM_User, time=0, value=1}
         {arriveTangent=-0.6219, interpMode=RCIM_Cubic, leaveTangent=-0.6219, tangentMode=RCTM_User, time=1, value=0}
        ]}}})
    Module.Param0WriteEnabled=true
    Module.Param1WriteEnabled=true
    Module.Param2WriteEnabled=true
    Module.Param3WriteEnabled=true
   )
  ]
  rendererClasses [&NiagaraRibbonRendererProperties]
  renderers [{rendererClass=&NiagaraRibbonRendererProperties}]
 }
}
summary {
 systemName=NS_Pickaxe_Guitar_NanaEternal_Trail
 emitters [
  {emitterName=WispTrail, rendererClasses [&NiagaraRibbonRendererProperties]}
  {emitterName=GLowTrail_CameraFacing, rendererClasses [&NiagaraRibbonRendererProperties]}
  {emitterName=GlowTrail, rendererClasses [&NiagaraRibbonRendererProperties]}
 ]
 userVariables [
  User.bIsFrontEnd {type=NiagaraBool}
  User.bIsFrontEndPreview {type=NiagaraBool}
  User.RibbonIDOffset {type=NiagaraInt32}
  User.SkeletalMesh {
   type=NiagaraDataInterfaceSkeletalMesh
   defaultValue {
    struct=NiagaraExt_VariableValue_DataInterface
    value {dataInterface=&self:NiagaraDataInterfaceSkeletalMesh_0}
   }
  }
 ]
}
